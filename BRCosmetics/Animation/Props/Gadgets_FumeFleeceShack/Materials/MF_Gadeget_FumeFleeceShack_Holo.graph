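# Material function that gives the FumeFleeceShack emote prop a hologram look, overriding base color and emissive on incoming material attributes.
# Emissive blends Fresnel-tinted colors through masks, adds a fake top light, and scales by time of day.
# Named reroutes supply flipbook and stepped noise, sweep, flicker, skylight cubemap and face masks, with pre-skinned position and normal available for mesh-locked projection.
# A sine wobble offset is built, but the world position offset output appears unconnected.
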MaterialFunction /BRCosmetics/Animation/Props/Gadgets_FumeFleeceShack/Materials/MF_Gadeget_FumeFleeceShack_Holo
graph {nodes [
  Constant()
  Frac()
  n1 = GetMaterialAttributes(None=FunctionInput() {inputName=IN MA, inputType=FunctionInput_MaterialAttributes}) {attributeGetTypes [#1, #2]}
  n2 = NamedRerouteDeclaration(None=@n1.BaseColor) {name=IN_Color, nodeColor=<rgba 0.3594 1 0 1>}
  n3 = VectorParameter() {group=FumeFleeceShack, parameterName=DarkerColor, defaultValue=<rgba 0.0334 0.0837 0.1562 1>}
  n4 = Multiply(A=@n3.RGB, B=NamedRerouteUsage())
  FunctionOutput(None=SetMaterialAttributes!(MaterialAttributes=@n1.MaterialAttributes, BaseColor=Multiply(A=@n2, B=@n4), EmissiveColor=NamedRerouteUsage(), WorldPositionOffset) {attributeSetTypes [#1, #3, #4]}) {outputName=OUT MA}
  LocalPosition() {localOrigin=InstancePreSkinning}
  n5 = Divide(A=MaterialFunctionCall(f3=MaterialFunctionCall() {materialFunction=&ObjectPivotPoint}.Object Pivot Location) {materialFunction=&AddComponents}.f3) {constB=128}
  n6 = MaterialFunctionCall() {materialFunction=&DitherTemporalAA}
  n7 = DotProduct(A=CameraVectorWS(), B=NamedRerouteUsage())
  Multiply(A=Frac(None=Add(A=@n5, B=@n6.Result)), B=OneMinus(None=Power(Base=@n7) {constExponent=8})) {constB=0.5}
  n8 = Constant3Vector() {desc=Random, constant=<rgba -0.3864 0.6 0.4464 0>}
  n9 = Add(
   A=Multiply(A=Time(), B=ScalarParameter() {defaultValue=4, group=WPO_Wobble, parameterName=Wobble_Speed, sortPriority=1})
   B=ScalarParameter() {group=WPO_Wobble, parameterName=Wobble_Offset, sortPriority=2}
  )
  n10 = ScreenPosition()
  n11 = MaterialFunctionCall(Float3=AppendVector(A=@n10.ViewportUV, B=ComponentMask(None=@n10.ViewportUV) {g=true})) {materialFunction=&BreakOutFloat3Components}
  n12 = NamedRerouteUsage()
  n13 = MaterialFunctionCall(Float3=Multiply(A=LinearInterpolate(Alpha=@n12) {constA=0.01, constB=0.1}) {constB=0.1}) {materialFunction=&BreakOutFloat3Components}
  n14 = Multiply(A=@n8.R, B=Add(A=@n9, B=Divide(A=@n11.R, B=@n13.R)))
  n15 = Multiply(A=@n8.G, B=Add(A=@n9, B=Divide(A=@n11.G, B=@n13.G)))
  n16 = Multiply(A=@n8.B, B=Add(A=@n9, B=Divide(A=@n11.B, B=@n13.B)))
  n17 = MaterialFunctionCall(Float3=MaterialFunctionCall(X=Sine(None=@n14) {period=6.283}, Y=Sine(None=@n15) {period=6.283}, Z=Sine(None=@n16) {period=6.283}) {materialFunction=&MakeFloat3}.Result) {materialFunction=&BreakOutFloat3Components}
  n18 = MaterialFunctionCall(X=Add(A=@n17.G, B=@n17.G), Y=Add(A=@n17.R, B=@n17.R), Z=Add(A=@n17.R, B=@n17.R)) {materialFunction=&MakeFloat3}
  n19 = Multiply(A=LinearInterpolate(Alpha=NamedRerouteUsage()) {constA=10}) {constB=0.1, desc=adjustStrength}
  NamedRerouteDeclaration(None=Multiply(A=Multiply(A=@n18.Result, B=@n19), B=ScalarParameter() {defaultValue=1, group=WPO_Wobble, parameterName=WPO_Factor})) {name=WPO_Wobble, nodeColor=<rgba 1 0 0.5391 1>}
  NamedRerouteDeclaration(None=ScalarParameter() {defaultValue=1, group=FumeFleeceShack, parameterName=StaticSpeed}) {name=StaticSpeed, nodeColor=<rgba 0 1 0.6484 1>}
  n20 = ComponentMask(None=NamedRerouteUsage()) {g=true}
  n21 = Time()
  n22 = Frac(None=Add(A=Multiply(A=@n20) {constB=0.75}, B=Multiply(A=@n21) {constB=-0.2}))
  NamedRerouteDeclaration(None=Multiply(A=Power(Base=@n22) {constExponent=20}, B=ScalarParameter() {defaultValue=0.01, group=FumeFleeceShack, parameterName=Sweep Add}) {constB=0.2}) {name=Sweep, nodeColor=<rgba 0.2422 0 1 1>}
  NamedRerouteDeclaration(None=ScreenPosition().ViewportUV) {name=StablizedUV, nodeColor=<rgba 0 1 0.2031 1>}
  n23 = MaterialFunctionCall!(Size=Constant2Vector() {g=16, r=16}, Selector=Divide(A=Time()) {constB=4}, UV=Multiply(A=NamedRerouteUsage()) {constB=80}) {materialFunction=/Game/Global/MaterialFunctions/MF_FlipbookUV}
  n24 = TextureSample(UVs=@n23.FlipbookUV1) {texture=/Game/Global/Textures/Default/Noise/Noise}
  n25 = TextureSample(UVs=Add(A=Multiply(A=@n23.FlipbookUV2) {constB=-0.8}) {constB=0.5}) {texture=&Noise}
  NamedRerouteDeclaration(None=Multiply(A=Multiply(A=@n24.R, B=@n25.R))) {name=Noise, nodeColor=<rgba 0.8516 1 0 1>}
  NamedRerouteDeclaration(None=VertexInterpolator(VS=LocalPosition() {localOrigin=InstancePreSkinning}.XYZ).PS) {name=PreSkinnedLoc, nodeColor=<rgba 0.5566 0.1791 0.7969 1>}
  n26 = NamedRerouteUsage()
  n27 = Multiply(A=Divide(A=Time()) {constB=1}) {constB=10}
  n28 = Multiply(A=Divide(A=Floor(None=@n27)) {constB=10}) {constB=8}
  n29 = MaterialFunctionCall(X=@n28, Y=Multiply(A=@n28) {constB=-1}, Z=@n28) {materialFunction=&MakeFloat3}
  NamedRerouteDeclaration(None=MaterialFunctionCall(
    TextureObject=TextureObject() {texture=/Game/Global/Textures/Default/Noise/TankMoteNoise}
    TextureSize=Constant() {r=128}
    WorldPosition=Subtract(A=@n26, B=@n29.Result)
    World Space Normal=NamedRerouteUsage()
   ) {materialFunction=&WorldAlignedTexture}.XYZ Texture) {name=NoiseColorMsk, nodeColor=<rgba 1 0 0.9141 1>}
  n30 = TextureSample() {samplerType=SAMPLERTYPE_Masks, texture=/BRCosmetics/Animation/Props/Gadgets_FumeFleeceShack/Textures/T_FumeFleeceShack_holo_Emote_Prop_S}
  NamedRerouteDeclaration(None=Step(Y=@n30.B) {constX=0.1, constY=0.8}) {name=FaceMsk, nodeColor=<rgba 0 0.25 1 1>}
  n31 = Multiply(A=NamedRerouteUsage())
  n32 = Multiply(A=NamedRerouteUsage()) {constB=2}
  n33 = NamedRerouteUsage()
  n34 = LinearInterpolate(
   A=Constant3Vector() {constant=<rgba 0.6042 0.1573 0.265 1>}
   B=Constant3Vector() {constant=<rgba 0.1461 0.459 0.5833 1>}
   Alpha=NamedRerouteUsage()
  )
  n35 = NamedRerouteUsage()
  n36 = Add(A=Multiply(A=Multiply(A=@n32, B=@n33), B=Multiply(A=@n34, B=@n35)), B=Multiply(A=@n34) {constB=0.1})
  NamedRerouteDeclaration(None=Multiply(A=Add(A=@n31, B=@n36), B=NamedRerouteUsage())) {name=Colored Noise, nodeColor=<rgba 0.5703 1 0 1>}
  n37 = Add(A=Multiply(A=CameraVectorWS()) {constB=-1}, B=Multiply(A=NamedRerouteUsage(), B=Constant() {r=0.5}))
  n38 = ScalarParameter() {defaultValue=0.8, group=FumeFleeceShack, parameterName=SkylightEnvMapRoughness}
  NamedRerouteDeclaration(None=Multiply(
    A=SkyLightEnvMapSample(Direction=@n37, Roughness=@n38)
    B=VectorParameter() {group=FumeFleeceShack, parameterName=CubeMap_Darken_Color, defaultValue=<rgba 0.1 0.1 0.1 1>}.RGB
   )) {name=Cubemap, nodeColor=<rgba 1 0 0.7266 1>}
  n39 = NamedRerouteUsage()
  n40 = ComponentMask(None=VectorParameter() {group=FumeFleeceShack, parameterName=TexCoordScale, defaultValue=<rgba 1 1 0 0>}.RGB) {g=true, r=true}
  n41 = VectorParameter() {group=FumeFleeceShack, parameterName=TexCoordOffset}
  NamedRerouteDeclaration(None=Add(A=Multiply(A=@n39, B=@n40), B=ComponentMask(None=@n41.RGB) {g=true, r=true})) {name=TexCoord, nodeColor=<rgba 1 0 0.4453 1>}
  n42 = Time() {bOverride_Period=true, period=2.5}
  n43 = ScalarParameter() {defaultValue=2, group=FumeFleeceShack, parameterName=Flicker Speed}
  n44 = Time() {bOverride_Period=true, period=0.4}
  n45 = Multiply(A=Sine(None=Multiply(A=@n42, B=@n43)), B=Sine(None=Multiply(A=@n44, B=@n43)))
  n46 = Sine(None=Multiply(A=Time() {bOverride_Period=true, period=1.13}, B=@n43))
  n47 = Multiply(A=Time() {bOverride_Period=true, period=1.5126}, B=@n43)
  n48 = LinearInterpolate(
   A=ScalarParameter() {defaultValue=0.79, group=FumeFleeceShack, parameterName=Flicker Intensity}
   Alpha=Multiply(A=Multiply(A=@n45, B=@n46), B=Sine(None=@n47))
  )
  NamedRerouteDeclaration(None=Saturate(None=Frac(None=@n48))) {name=Flicker, nodeColor=<rgba 0 1 1 1>}
  n49 = CameraPositionWS()
  n50 = WorldPosition()
  n51 = Divide(A=Length(None=Subtract(A=@n49, B=@n50.XYZ)) {is=class MaterialExpressionLength}, B=Multiply(A=Constant() {r=16}) {constB=100})
  NamedRerouteDeclaration(None=OneMinus(None=Saturate(None=@n51))) {name=DistToFade, nodeColor=<rgba 0.4531 1 0 1>}
  n52 = Subtract(A=TextureCoordinate()) {constB=0.5}
  n53 = Constant2Vector() {g=-0.1}
  n54 = Length(None=Multiply(A=Add(A=@n52, B=@n53), B=Constant2Vector() {g=1, r=1}))
  NamedRerouteDeclaration(None=Max(A=SmoothStep(Value=@n54) {constMax=0, constMin=0.1}, B=Step(Y=@n30.G) {constX=0.5, constY=0.5})) {name=InnerMouseMsk, nodeColor=<rgba 0 0.25 1 1>}
  n55 = Add(A=NamedRerouteUsage(), B=@n3.RGB)
  n56 = VectorParameter() {group=FumeFleeceShack, parameterName=LighterColor, defaultValue=<rgba 0.226 0.6744 0.8854 1>}
  n57 = CameraVectorWS()
  n58 = VertexNormalWS()
  n59 = Saturate(None=Power(Base=DotProduct(A=@n57, B=@n58), Exp=ScalarParameter() {defaultValue=4, parameterName=HeadFresnelExp}) {constExponent=4})
  n60 = LinearInterpolate(A=@n1.BaseColor, B=Add(A=LinearInterpolate(A=@n55, B=@n56.RGB, Alpha=@n59), B=NamedRerouteUsage()), Alpha=NamedRerouteUsage())
  n61 = DotProduct(A=Constant3Vector() {constant=<rgba 0 0 1 0>}, B=NamedRerouteUsage())
  n62 = Saturate(None=Power(Base=Abs(None=@n61)) {constExponent=8})
  n63 = VectorParameter() {group=FumeFleeceShack, parameterName=LighterColor, defaultValue=<rgba 0.226 0.6744 0.8854 1>}
  n64 = Multiply(
   A=Multiply(A=Multiply(A=@n62, B=@n63.RGB), B=ScalarParameter() {defaultValue=2, parameterName=FakeLightIntensity})
   B=OneMinus(None=NamedRerouteUsage())
  )
  n65 = VectorParameter() {group=FumeFleeceShack, parameterName=TODEmissive, defaultValue=<rgba 1 1 0.5 0.5>}
  NamedRerouteDeclaration(None=Multiply(
    A=Add(A=@n60, B=@n64)
    B=MaterialFunctionCall(Value_Morning=@n65.R, Value_Day=@n65.R, Value_Evening=@n65.R, Value_Night=@n65.R) {materialFunction=&Scalar_Time_of_Day}.Result
   )) {name=OUT_Emissive, nodeColor=<rgba 1 0 0.6797 1>}
  NamedRerouteDeclaration(None=VertexInterpolator(VS=PreSkinnedNormal()).PS) {name=PreskinnedNormal, nodeColor=<rgba 0.5566 0.1791 0.7969 1>}
  NamedRerouteUsage()
  ScalarParameter() {defaultValue=1, group=WPO_Wobble, parameterName=Wobble_Strength, sortPriority=3}
  ScalarParameter() {defaultValue=1, group=WPO_Wobble, parameterName=Wobble_Amplitude, sortPriority=4}
  ScalarParameter() {defaultValue=4, group=WPO_Wobble, parameterName=Wobble_Scale, sortPriority=0}
  VertexNormalWS()
 ]}
