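# Material for the Fume Fleece Shack gadget device, layering an animated emissive tank onto the shared 2023 character PreFX/PostFX functions.
# A UV mask selects the tank, which blends a two-color bounding-box gradient, Z-scrolling swirl noise and a Fresnel rim, and is pushed slightly along its normals.
# Cheap reflections sample the skylight, or a lat-long interior texture on forward and mobile.
# Time of day scales final emissive.
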Material /BRCosmetics/Animation/Props/Gadgets_FumeFleeceShack/Materials/M_Gadget_FumeFleeceShack_Device
graph {
 properties {
  bUseMaterialAttributes=true
  bUsedWithMeshDeformer=true
  bUsedWithMorphTargets=true
  bUsedWithSkeletalMesh=true
  pixelDepthOffsetMode=PDOM_Legacy
  shadingModel=MSM_FromMaterialExpression
  usedShadingModels=Default Lit | Subsurface | Clear Coat | Subsurface Profile | Cloth
 }
 nodes [
  Constant()
  n1 = MaterialFunctionCall() {materialFunction=&MF_CharacterParent_2023_PreFX}
  n2 = GetMaterialAttributes(None=@n1.Result) {attributeGetTypes [#1]}
  n3 = VectorParameter() {group=FumeFleeceShack, parameterName=Tank_Color_B, defaultValue=<rgba 0.0841 0.3308 0.4896 1>}
  n4 = MaterialFunctionCall() {materialFunction=&BoundingBoxBased_0-1_UVW}
  n5 = Multiply(A=OneMinus(None=@n4.B), B=@n4.B)
  n6 = Constant4Vector() {constant=<rgba 0 1 0.24 0.87>}
  n7 = SmoothStep(Value=Power(Base=MaterialFunctionCall(Input=@n5, Input Low=@n6.R, Input High=@n6.R, Target Low=@n6.R, Target High=@n6.R) {materialFunction=&RemapValueRange}.Result) {constExponent=0.25}) {constMax=0.8}
  n8 = Multiply(A=Time()) {constB=64}
  n9 = MaterialFunctionCall(
   TextureObject=TextureObject() {samplerType=SAMPLERTYPE_LinearGrayscale, texture=/Game/Global/Textures/Default/Noise/T_SwirlNoise_H2}
   TextureSize=Constant3Vector() {constant=<rgba 48 48 512 0>}
   WorldPosition=Subtract(
    A=NamedRerouteUsage()
    B=MaterialFunctionCall(Z=@n8) {materialFunction=&MakeFloat3}.Result
   )
  ) {materialFunction=&WorldAlignedTexture}
  n10 = Add(
   A=LinearInterpolate(
    A=@n3.RGB
    B=VectorParameter() {group=FumeFleeceShack, parameterName=Tank_Color_A, defaultValue=<rgba 0.021 0.0551 0.1302 1>}.RGB
    Alpha=Saturate(None=@n7)
   )
   B=Multiply(A=@n3.RGB, B=Multiply(A=@n9.XY Texture) {constB=0.5}) {constB=0.5}
  )
  n11 = NamedRerouteUsage()
  n12 = VertexNormalWS()
  n13 = Constant3Vector() {constant=<rgba 3 3 1 0>}
  n14 = SmoothStep(Value=Fresnel(Normal=Multiply(A=@n12, B=@n13)) {baseReflectFraction=0.01, exponent=0.2})
  n15 = Step(Y=ComponentMask(None=TextureCoordinate()) {g=true}) {constX=0.4}
  n16 = LinearInterpolate(A=@n2.Emissive Color, B=Add(A=Add(A=@n10, B=@n11), B=OneMinus(None=@n14)), Alpha=@n15)
  n17 = VectorParameter() {group=FumeFleeceShack, parameterName=TODEmissive, defaultValue=<rgba 1 1 0.6 0.6>}
  n18 = MaterialFunctionCall(Value_Morning=@n17.R, Value_Day=@n17.R, Value_Evening=@n17.R, Value_Night=@n17.R) {materialFunction=&Scalar_Time_of_Day}
  n19 = Multiply(A=VertexNormalWS()) {constB=0.2}
  MaterialFunctionCall(
   Material=SetMaterialAttributes!(MaterialAttributes=@n2.MaterialAttributes, Emissive Color=Multiply(A=@n16, B=@n18.Result), World Position Offset=Multiply(A=@n15, B=@n19)) {attributeSetTypes [#1, #2]}
   M=@n1.M
   Pre Skinned Local Position=@n1.M
   Distance Blend=@n1.M
   Fresnel=@n1.M
  ) {materialFunction=&MF_CharacterParent_2023_PostFX}
  NamedRerouteDeclaration(None=VertexInterpolator(VS=LocalPosition() {localOrigin=InstancePreSkinning}.XYZ).PS) {name=PreSkinnedLoc, nodeColor=<rgba 0.5913 0.2783 1 1>}
  NamedRerouteDeclaration(None=VertexInterpolator(VS=PreSkinnedNormal()).PS) {name=PreskinnedNormal, nodeColor=<rgba 0.5913 0.2783 1 1>}
  n20 = CameraVectorWS()
  n21 = VertexNormalWS()
  n22 = Constant() {r=0.5}
  n23 = SkyLightEnvMapSample(
   Direction=Add(A=Multiply(A=@n20) {constB=-1}, B=Multiply(A=@n21, B=@n22))
   Roughness=ScalarParameter() {group=FumeFleeceShack, parameterName=SkylightEnvMapRoughness}
  )
  n24 = TextureSample(UVs=MaterialFunctionCall!(Vector=ReflectionVectorWS()) {materialFunction=/Engine/Functions/Engine_MaterialFunctions01/Coordinates/LongLatToUV}.Result) {texture=/Game/Environments/Asteria/Sets/NeonCity/Techart/Textures/T_FakeInterior_03_Blur_LongLat}
  NamedRerouteDeclaration(None=Multiply(
    A=ShadingPathSwitch(Default=@n23, Forward=@n24.RGB, Mobile=@n24.RGB)
    B=VectorParameter() {group=FumeFleeceShack, parameterName=CubeMap_Darken_Color, defaultValue=<rgba 0.1 0.1 0.1 1>}.RGB
   )) {name=Cubemap, nodeColor=<rgba 1 0 0.7266 1>}
 ]
}
