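# Translucent glass tank material for the FumeFleeceShack gadget, built on shared glass and thin-film functions with optional camera dither and vertex hide.
# Emissive fakes refraction by offsetting scene color with masked distortion, then adds a two-color tank glow with scrolling swirl noise, scaled by time of day.
# Forward and mobile paths substitute a lat-long fake reflection and depth-faded opacity for the scene-color sample and full opacity.
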Material /BRCosmetics/Animation/Props/Gadgets_FumeFleeceShack/Materials/M_Gadget_FumeFleeceShack_Glass
graph {
 properties {
  bAutomaticallySetUsageInEditor=false
  bUseMaterialAttributes=true
  bUsedWithMeshDeformer=true
  bUsedWithNiagaraRibbons=true
  bUsedWithNiagaraSprites=true
  bUsedWithSkeletalMesh=true
  blendMode=BLEND_Translucent
  translucencyPass=MTP_BeforeDOF
  usedShadingModels=Default Lit
 }
 nodes [
  Constant()
  n1 = TextureSampleParameter2D() {
   group=00_Glass
   parameterName=TechArtMask
   samplerType=SAMPLERTYPE_Masks
   sortPriority=-96
   texture=&T_BlackMask
  }
  n2 = MaterialFunctionCall!(
   Material Attributes=MaterialFunctionCall!(TechArtMask=@n1.RGB) {materialFunction=/Game/Global/Materials/Character_Parent_2023/Glass_Parent_2023/MF_Glass}.Result
   Fresnel=Fresnel() {baseReflectFraction=0, exponent=1}
   Alpha=@n1.RGB
  ) {materialFunction=/Game/Global/MaterialFunctions/TechArt/MF_ThinFilm_TechArt}
  n3 = StaticSwitchParameter(
   True=MaterialFunctionCall!(In=@n2.Result) {materialFunction=/Game/Packages/Fortress_SharedMaterials/Base_Material_Functions/BaseCustomizationFunctions/MF_FN_Vehicle_CameraDither}.Result
   False=@n2.Result
  ) {group=00_Glass, parameterName=UseCameraDither, sortPriority=250}
  n4 = ObjectPositionWS()
  n5 = WorldPosition()
  n6 = SetMaterialAttributes!(MaterialAttributes=@n3, World Position Offset=Multiply(A=Subtract(A=@n4, B=@n5.XYZ))) {attributeSetTypes [#1]}
  n7 = MaterialFunctionCall!(Material=@n3) {materialFunction=/Game/Global/Materials/Character_Parent_2023/Functions/BaseFN_CharacterOpacity_2023}
  n8 = Saturate(None=MaterialFunctionCall(
    Axis=Constant3Vector() {constant=<rgba 0 0 1 0>}
    Exponent=ScalarParameter() {defaultValue=5, parameterName=CylinderDistortionFresnelMaskExponent}
   ) {materialFunction=&AxisAlignedFresnel}.Fresnel)
  n9 = NamedRerouteUsage()
  n10 = Constant3Vector() {constant=<rgba 0 0 60 0>}
  n11 = Multiply(A=Power(Base=SphereMask(A=@n9, B=@n10) {attenuationRadius=36, hardnessPercent=0})) {constB=2}
  n12 = TransformPosition(None=ActorPositionWS()) {transformSourceType=TRANSFORMPOSSOURCE_World, transformType=TRANSFORMPOSSOURCE_View}
  n13 = Multiply(A=NamedRerouteUsage(), B=Divide(A=Multiply(A=@n8, B=@n11), B=ComponentMask(None=@n12) {b=true})) {constB=0.05}
  n14 = Multiply(A=@n13, B=LinearInterpolate(B=ScalarParameter() {defaultValue=0.8, parameterName=DistortionStrength}, Alpha=NamedRerouteUsage()) {constB=0})
  n15 = SceneColor!(UVs=Add(
    A=ScreenPosition().ViewportUV
    B=MaterialFunctionCall(X=@n14) {materialFunction=&MakeFloat2}.Result
   ))
  n16 = Constant() {r=4}
  n17 = ReflectionVectorWS(CustomWorldNormal=@n13)
  n18 = Multiply(
   A=TextureSample(UVs=MaterialFunctionCall!(Vector=@n17) {materialFunction=/Engine/Functions/Engine_MaterialFunctions01/Coordinates/LongLatToUV}.Result) {texture=/Game/Environments/Asteria/Sets/NeonCity/Techart/Textures/T_FakeInterior_03_Blur_LongLat}.RGB
   B=VectorParameter() {group=FumeFleeceShack, parameterName=CubeMap_Darken_Color, defaultValue=<rgba 0.1 0.1 0.1 1>}.RGB
  ) {constB=10}
  n19 = VectorParameter() {group=FumeFleeceShack, parameterName=Tank_Color_B, defaultValue=<rgba 0.1779 0.5972 0.7991 1>}
  n20 = VectorParameter() {group=FumeFleeceShack, parameterName=Tank_Color_A, defaultValue=<rgba 0.0446 0.1169 0.276 1>}
  n21 = MaterialFunctionCall() {materialFunction=&BoundingBoxBased_0-1_UVW}
  n22 = Constant4Vector() {constant=<rgba 0 1 0.24 0.87>}
  n23 = MaterialFunctionCall(Input=Multiply(A=OneMinus(None=@n21.B), B=@n21.B), Input Low=@n22.R, Input High=@n22.R, Target Low=@n22.R, Target High=@n22.R) {materialFunction=&RemapValueRange}
  n24 = Saturate(None=SmoothStep(Value=Power(Base=@n23.Result) {constExponent=0.25}) {constMax=0.8})
  n25 = NamedRerouteUsage()
  n26 = MaterialFunctionCall(Z=Multiply(A=Time()) {constB=32}) {materialFunction=&MakeFloat3}
  n27 = Multiply(A=MaterialFunctionCall(
    TextureObject=TextureObject() {samplerType=SAMPLERTYPE_LinearGrayscale, texture=/Game/Global/Textures/Default/Noise/T_SwirlNoise_H2}
    TextureSize=Constant3Vector() {constant=<rgba 32 32 512 0>}
    WorldPosition=Subtract(A=@n25, B=@n26.Result)
   ) {materialFunction=&WorldAlignedTexture}.XY Texture) {constB=0.5}
  n28 = Multiply(A=NamedRerouteUsage(), B=Constant3Vector() {constant=<rgba 3 3 1 0>})
  n29 = OneMinus(None=SmoothStep(Value=Fresnel(Normal=@n28) {baseReflectFraction=0.01, exponent=0.2}))
  n30 = Add(
   A=Add(A=LinearInterpolate(A=@n19.RGB, B=@n20.RGB, Alpha=@n24), B=Multiply(A=@n19.RGB, B=@n27) {constB=0.5})
   B=Multiply(A=Saturate(None=@n29)) {constB=0.1}
  )
  n31 = DotProduct(A=CameraVectorWS(), B=NamedRerouteUsage())
  n32 = LinearInterpolate(
   A=ScalarParameter() {defaultValue=0.2, parameterName=Glow_B}
   B=ScalarParameter() {defaultValue=1, parameterName=Glow_A}
   Alpha=Power(Base=OneMinus(None=@n31))
  )
  n33 = Add(
   A=ShadingPathSwitch(Default=Multiply(A=@n15.RGB, B=@n16) {constB=4}, Forward=@n18, Mobile=@n18)
   B=Multiply(A=Multiply(A=@n30, B=@n32)) {constB=0.1}
  )
  n34 = VectorParameter() {parameterName=TODEmissive, defaultValue=<rgba 0.125 0.125 0.25 0.25>}
  n35 = MaterialFunctionCall(Value_Morning=@n34.R, Value_Day=@n34.R, Value_Evening=@n34.R, Value_Night=@n34.R) {materialFunction=&Scalar_Time_of_Day}
  n36 = Constant() {r=1}
  n37 = DepthFade(FadeDistance=ScalarParameter() {defaultValue=20, group=FumeFleeceShack, parameterName=DepthFade Glow})
  n38 = LinearInterpolate(
   A=ScalarParameter() {defaultValue=0.168, group=FumeFleeceShack, parameterName=OpacityB}
   B=ScalarParameter() {defaultValue=0.324, group=FumeFleeceShack, parameterName=OpacityA}
   Alpha=Multiply(B=OneMinus(None=@n37)) {constA=1}
  )
  n39 = VertexNormalWS()
  MaterialFunctionCall(In=SetMaterialAttributes!(
    MaterialAttributes=StaticSwitchParameter(True=@n6, False=@n7.Result) {group=00_Glass, parameterName=UseVertexHide, sortPriority=300}
    Emissive Color=Multiply(A=@n33, B=@n35.Result)
    Opacity=ShadingPathSwitch(Default=@n36, Forward=@n38, Mobile=@n38)
    World Position Offset=Multiply(A=@n39) {constB=0.1}
   ) {attributeSetTypes [#2, #3, #1]}) {materialFunction=&MF_TechArtOverrides}
  NamedRerouteDeclaration(None=VertexInterpolator(VS=LocalPosition() {localOrigin=InstancePreSkinning}.XYZ).PS) {name=PreSkinnedLoc, nodeColor=<rgba 0.5913 0.2783 1 1>}
  NamedRerouteDeclaration(None=VertexInterpolator(VS=PreSkinnedNormal()).PS) {name=PreskinnedNormal, nodeColor=<rgba 0.5913 0.2783 1 1>}
  n40 = Add(A=Multiply(A=CameraVectorWS()) {constB=-1}, B=Multiply(A=VertexNormalWS(), B=Constant() {r=0.5}))
  n41 = ScalarParameter() {group=FumeFleeceShack, parameterName=SkylightEnvMapRoughness}
  NamedRerouteDeclaration(None=Multiply(
    A=SkyLightEnvMapSample(Direction=@n40, Roughness=@n41)
    B=VectorParameter() {group=FumeFleeceShack, parameterName=CubeMap_Darken_Color, defaultValue=<rgba 0.1 0.1 0.1 1>}.RGB
   )) {name=Cubemap, nodeColor=<rgba 1 0 0.7266 1>}
  n42 = Multiply(A=Time(), B=ScalarParameter() {defaultValue=1, group=FumeFleeceShack, parameterName=Sine Z Pan Speed})
  n43 = MaterialFunctionCall() {materialFunction=&ObjectPivotPoint}
  n44 = NamedRerouteUsage()
  n45 = Divide(
   A=Subtract(A=ComponentMask(None=@n43.Object Pivot Location) {b=true}, B=ComponentMask(None=@n44) {b=true})
   B=ScalarParameter() {defaultValue=5, group=FumeFleeceShack, parameterName=Sine Z Frequency}
  )
  NamedRerouteDeclaration(None=Sine(None=Add(A=@n42, B=@n45))) {name=Sinewave_Wobble_Distort, nodeColor=<rgba 1 0 0.5391 1>}
  n46 = Length(None=Subtract(A=CameraPositionWS(), B=WorldPosition().XYZ)) {is=class MaterialExpressionLength}
  n47 = Multiply(A=Constant() {r=1.8}) {constB=100}
  n48 = Saturate(None=Power(Base=Divide(A=@n46, B=@n47)) {constExponent=4})
  NamedRerouteDeclaration(None=@n48) {name=DistToFade, nodeColor=<rgba 0.4531 1 0 1>}
  ObjectPositionWS()
  OneMinus(None=@n48)
  VertexNormalWS()
  WorldPosition()
 ]
}
